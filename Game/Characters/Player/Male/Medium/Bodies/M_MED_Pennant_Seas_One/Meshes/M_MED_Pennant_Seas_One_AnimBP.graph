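# Animation Blueprint for the male medium Pennant Seas One player body, adding deform correction and zipper motion.
# It copies the attached parent's pose into the female variant's deform Control Rig, then simulates the zipper bone with gravity-driven AnimDynamics.
# A weighted constraint ties the master zipper joint to upper spine, neck and head bones.
# A front-end flag, converted to alpha each update, applies an additive yaw to the dynamic zipper bone.
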AnimBlueprint /Game/Characters/Player/Male/Medium/Bodies/M_MED_Pennant_Seas_One/Meshes/M_MED_Pennant_Seas_One_AnimBP
target_skeleton=&/M_MED_Pennant_Seas_One_Skeleton
graphs [
 AnimGraph {nodes [
   n1 = Control Rig(Source=Copy Pose From Mesh() {bUseAttachedParent=true}) {
    alphaBoolBlend {blendOption=Linear}
    controlRigAssetReference {blueprintRigClass=/Game/Characters/Player/Female/Medium/Bodies/F_MED_Pennant_Seas_One/Meshes/F_MED_Pennant_Seas_One_DeformRig_ControlRig.F_MED_Pennant_Seas_One_DeformRig_ControlRig_C}
   }
   n2 = Anim Dynamics - Bone: dyn_zipper_jnt(ComponentPose=Local To Component(LocalPose=@n1), Alpha=1, ExternalForce="0,0,0") {
    bUseGravityOverride=true
    alphaBoolBlend {blendOption=Linear}
    boundBone {boneName=dyn_zipper_jnt}
    componentAppliedLinearAccClamp=<3000 3000 3000>
    componentLinearAccScale=<0.4 0.3 0.3>
    componentLinearVelScale=<0.4 0.3 0.3>
    gravityOverride {z=-980}
    physicsBodyDefinitions [{
      boundBone {boneName=dyn_zipper_jnt}
      boxExtents=<10 10 10>
      constraintSetup {angularLimitsMax=<5 5 80>, angularLimitsMin=<-5 -5 0>, angularTarget=<1 0 0>}
      localJointOffset=<5 0 0>
     }]
    relativeSpaceBone {boneName=head}
   }
   n3 = "Constraint - master_zipper_jnt (PPPP)"!(
    ComponentPose=@n2
    Alpha=1
    ConstraintWeights_0=0.3022
    ConstraintWeights_1:float=0.0566
    ConstraintWeights_2:float=0.6208
    ConstraintWeights_3:float=0.0203
   ) -> (Pose) {
    alphaBoolBlend {blendOption=Linear}
    boneToModify {boneName=master_zipper_jnt}
    constraintSetup [{targetBone {boneName=spine_05}}, {targetBone {boneName=neck_01}}, {targetBone {boneName=neck_02}}, {targetBone {boneName=head}}]
    constraintWeights [1,1,1,1]
   }
   n4 = "Transform (Modify) Bone - Bone: dyn_zipper_jnt"(ComponentPose=@n3, Alpha=Get front end() -> (front end:double), Translation="0,0,0", Rotation="0,20,0", Scale="1,1,1") {
    rotationMode=BMM_Additive
    rotationSpace=BCS_BoneSpace
    alphaBoolBlend {blendOption=Linear}
    boneToModify {boneName=dyn_zipper_jnt}
    rotation {yaw=20}
   }
   Output Pose(Result=Component To Local(ComponentPose=@n4)) {layerGroup=None}
  ]}
 EventGraph {nodes [
   Event BlueprintUpdateAnimation()
   > Set front end(front end:double="To Float (Boolean)"(InBool=Get bIsInFrontEnd())) -> (Output_Get:double)
  ]}
]
variables [front end {type=number}]
